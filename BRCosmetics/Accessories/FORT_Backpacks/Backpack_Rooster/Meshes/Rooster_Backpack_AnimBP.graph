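# Animation Blueprint for the Rooster backpack cosmetic that poses its six wing feather bones procedurally, with no animation clips.
# A wingPosition state machine switches between Base, Jumping, Floating, Targeting and Diving using jump, aim, skydive, parachute and ground flags.
# Each state applies additive bone-space yaw to the reference pose, except Base and Targeting, which leave it untouched.
# Afterwards, PawnSpeed2D drives a clamped, range-mapped alpha that adds further feather yaw with movement speed.
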AnimBlueprint /BRCosmetics/Accessories/FORT_Backpacks/Backpack_Rooster/Meshes/Rooster_Backpack_AnimBP
target_skeleton=&/Rooster_Backpack_Skeleton
graphs [
 AnimGraph {nodes [
   AnimGraphNode_StateMachine_0 = wingPosition() {is=title AnimGraphNode_StateMachine}
   n1 = Get PawnSpeed2D()
   n2 = "Transform (Modify) Bone - Bone: r_wing_feather_top_01"(ComponentPose=Local To Component(LocalPose=@AnimGraphNode_StateMachine_0), Alpha=@n1, Translation="0,0,0", Rotation="0,15,0", Scale="1,1,1") {
    rotationMode=BMM_Additive
    rotationSpace=BCS_BoneSpace
    alphaBoolBlend {blendOption=Linear}
    alphaScaleBiasClamp {bClampResult=true, bMapRange=true, clampMax=0.75, inRange {max=549}, outRange {max=0.75}}
    boneToModify {boneName=r_wing_feather_top_01}
    rotation {yaw=15}
   }
   n3 = "Transform (Modify) Bone - Bone: r_wing_feather_mid_01"(ComponentPose=@n2, Alpha=@n1, Translation="0,0,0", Rotation="0,15,0", Scale="1,1,1") {
    rotationMode=BMM_Additive
    rotationSpace=BCS_BoneSpace
    alphaBoolBlend {blendOption=Linear}
    alphaScaleBiasClamp {bClampResult=true, bMapRange=true, clampMax=0.75, inRange {max=549}, outRange {max=0.75}}
    boneToModify {boneName=r_wing_feather_mid_01}
    rotation {yaw=15}
   }
   n4 = "Transform (Modify) Bone - Bone: r_wing_feather_bot_01"(ComponentPose=@n3, Alpha=@n1, Translation="0,0,0", Rotation="0,15,0", Scale="1,1,1") {
    rotationMode=BMM_Additive
    rotationSpace=BCS_BoneSpace
    alphaBoolBlend {blendOption=Linear}
    alphaScaleBiasClamp {bClampResult=true, bMapRange=true, clampMax=0.75, inRange {max=549}, outRange {max=0.75}}
    boneToModify {boneName=r_wing_feather_bot_01}
    rotation {yaw=15}
   }
   n5 = "Transform (Modify) Bone - Bone: l_wing_feathel_top_01"(ComponentPose=@n4, Alpha=@n1, Translation="0,0,0", Rotation="0,15,0", Scale="1,1,1") {
    rotationMode=BMM_Additive
    rotationSpace=BCS_BoneSpace
    alphaBoolBlend {blendOption=Linear}
    alphaScaleBiasClamp {bClampResult=true, bMapRange=true, clampMax=0.75, inRange {max=549}, outRange {max=0.75}}
    boneToModify {boneName=l_wing_feathel_top_01}
    rotation {yaw=15}
   }
   n6 = "Transform (Modify) Bone - Bone: l_wing_feathel_mid_01"(ComponentPose=@n5, Alpha=@n1, Translation="0,0,0", Rotation="0,15,0", Scale="1,1,1") {
    rotationMode=BMM_Additive
    rotationSpace=BCS_BoneSpace
    alphaBoolBlend {blendOption=Linear}
    alphaScaleBiasClamp {bClampResult=true, bMapRange=true, clampMax=0.75, inRange {max=549}, outRange {max=0.75}}
    boneToModify {boneName=l_wing_feathel_mid_01}
    rotation {yaw=15}
   }
   n7 = "Transform (Modify) Bone - Bone: l_wing_feathel_bot_01"(ComponentPose=@n6, Alpha=@n1, Translation="0,0,0", Rotation="0,15,0", Scale="1,1,1") {
    rotationMode=BMM_Additive
    rotationSpace=BCS_BoneSpace
    alphaBoolBlend {blendOption=Linear}
    alphaScaleBiasClamp {bClampResult=true, bMapRange=true, clampMax=0.75, inRange {max=549}, outRange {max=0.75}}
    boneToModify {boneName=l_wing_feathel_bot_01}
    rotation {yaw=15}
   }
   Output Pose(Result=Component To Local(ComponentPose=@n7)) {layerGroup=None}
  ]}
 wingPosition {
  entry=Base
  path=AnimGraph.AnimGraphNode_StateMachine_0.wingPosition
  ids=id
  nodes [
   AnimStateNode_0 = Base() {is=title AnimStateNode}
   AnimStateNode_1 = Jumping() {is=title AnimStateNode}
   AnimStateNode_2 = Floating() {is=title AnimStateNode}
   AnimStateNode_3 = Targeting() {is=title AnimStateNode}
   AnimStateNode_4 = Diving() {is=title AnimStateNode}
   AnimStateTransitionNode_1 = AnimStateTransitionNode() {priorityOrder=3}
   AnimStateTransitionNode_2 = AnimStateTransitionNode() {priorityOrder=2}
   AnimStateTransitionNode_3 = AnimStateTransitionNode()
   AnimStateTransitionNode_4 = AnimStateTransitionNode()
   AnimStateTransitionNode_5 = AnimStateTransitionNode()
   AnimStateTransitionNode_6 = AnimStateTransitionNode()
   AnimStateTransitionNode_7 = AnimStateTransitionNode()
   AnimStateTransitionNode_8 = AnimStateTransitionNode()
   AnimStateTransitionNode_9 = AnimStateTransitionNode()
  ]
  links [
   {
    dst=Diving
    rule=AnimGraph.AnimGraphNode_StateMachine_0.wingPosition.AnimStateTransitionNode_3.Transition
    src=Base
    via=AnimStateTransitionNode_3
   }
   {
    dst=Jumping
    rule=AnimGraph.AnimGraphNode_StateMachine_0.wingPosition.AnimStateTransitionNode_8.Transition
    src=Base
    via=AnimStateTransitionNode_8
   }
   {
    dst=Targeting
    rule=AnimGraph.AnimGraphNode_StateMachine_0.wingPosition.AnimStateTransitionNode_6.Transition
    src=Base
    via=AnimStateTransitionNode_6
   }
   {
    dst=Floating
    rule=AnimGraph.AnimGraphNode_StateMachine_0.wingPosition.AnimStateTransitionNode_4.Transition
    src=Diving
    via=AnimStateTransitionNode_4
   }
   {
    dst=Base
    rule=AnimGraph.AnimGraphNode_StateMachine_0.wingPosition.AnimStateTransitionNode_1.Transition
    src=Floating
    via=AnimStateTransitionNode_1
   }
   {
    dst=Base
    rule=AnimGraph.AnimGraphNode_StateMachine_0.wingPosition.AnimStateTransitionNode_2.Transition
    src=Floating
    via=AnimStateTransitionNode_2
   }
   {
    dst=Diving
    rule=AnimGraph.AnimGraphNode_StateMachine_0.wingPosition.AnimStateTransitionNode_5.Transition
    src=Floating
    via=AnimStateTransitionNode_5
   }
   {
    dst=Base
    rule=AnimGraph.AnimGraphNode_StateMachine_0.wingPosition.AnimStateTransitionNode_9.Transition
    src=Jumping
    via=AnimStateTransitionNode_9
   }
   {
    dst=Base
    rule=AnimGraph.AnimGraphNode_StateMachine_0.wingPosition.AnimStateTransitionNode_7.Transition
    src=Targeting
    via=AnimStateTransitionNode_7
   }
  ]
 }
 Base {path=AnimGraph.AnimGraphNode_StateMachine_0.wingPosition.AnimStateNode_0.Base, nodes [Output Animation Pose(Result=Local Space Ref Pose())]}
 Jumping {
  path=AnimGraph.AnimGraphNode_StateMachine_0.wingPosition.AnimStateNode_1.Jumping
  nodes [
   n1 = "Transform (Modify) Bone - Bone: r_wing_feather_top_01"(ComponentPose=Local To Component(LocalPose=Local Space Ref Pose()), Alpha=0.25, Translation="0,0,0", Rotation="0,-60,0", Scale="1,1,1") {
    rotationMode=BMM_Additive
    rotationSpace=BCS_BoneSpace
    alphaBoolBlend {blendOption=Linear}
    boneToModify {boneName=r_wing_feather_top_01}
    rotation {yaw=-60}
   }
   n2 = "Transform (Modify) Bone - Bone: r_wing_feather_mid_01"(ComponentPose=@n1, Alpha=0.25, Translation="0,0,0", Rotation="0,-40,0", Scale="1,1,1") {
    rotationMode=BMM_Additive
    rotationSpace=BCS_BoneSpace
    alphaBoolBlend {blendOption=Linear}
    boneToModify {boneName=r_wing_feather_mid_01}
    rotation {yaw=-40}
   }
   n3 = "Transform (Modify) Bone - Bone: r_wing_feather_bot_01"(ComponentPose=@n2, Alpha=0.25, Translation="0,0,0", Rotation="0,-20,0", Scale="1,1,1") {
    rotationMode=BMM_Additive
    rotationSpace=BCS_BoneSpace
    alphaBoolBlend {blendOption=Linear}
    boneToModify {boneName=r_wing_feather_bot_01}
    rotation {yaw=-20}
   }
   n4 = "Transform (Modify) Bone - Bone: l_wing_feathel_top_01"(ComponentPose=@n3, Alpha=0.25, Translation="0,0,0", Rotation="0,-60,0", Scale="1,1,1") {
    rotationMode=BMM_Additive
    rotationSpace=BCS_BoneSpace
    alphaBoolBlend {blendOption=Linear}
    boneToModify {boneName=l_wing_feathel_top_01}
    rotation {yaw=-60}
   }
   n5 = "Transform (Modify) Bone - Bone: l_wing_feathel_mid_01"(ComponentPose=@n4, Alpha=0.25, Translation="0,0,0", Rotation="0,-40,0", Scale="1,1,1") {
    rotationMode=BMM_Additive
    rotationSpace=BCS_BoneSpace
    alphaBoolBlend {blendOption=Linear}
    boneToModify {boneName=l_wing_feathel_mid_01}
    rotation {yaw=-40}
   }
   n6 = "Transform (Modify) Bone - Bone: l_wing_feathel_bot_01"(ComponentPose=@n5, Alpha=0.25, Translation="0,0,0", Rotation="0,-20,0", Scale="1,1,1") {
    rotationMode=BMM_Additive
    rotationSpace=BCS_BoneSpace
    alphaBoolBlend {blendOption=Linear}
    boneToModify {boneName=l_wing_feathel_bot_01}
    rotation {yaw=-20}
   }
   Output Animation Pose(Result=Component To Local(ComponentPose=@n6))
  ]
 }
 Floating {
  path=AnimGraph.AnimGraphNode_StateMachine_0.wingPosition.AnimStateNode_2.Floating
  nodes [
   n1 = "Transform (Modify) Bone - Bone: r_wing_feather_top_01"(ComponentPose=Local To Component(LocalPose=Local Space Ref Pose()), Alpha=1, Translation="0,0,0", Rotation="0,-60,0", Scale="1,1,1") {
    rotationMode=BMM_Additive
    rotationSpace=BCS_BoneSpace
    alphaBoolBlend {blendOption=Linear}
    boneToModify {boneName=r_wing_feather_top_01}
    rotation {yaw=-60}
   }
   n2 = "Transform (Modify) Bone - Bone: r_wing_feather_mid_01"(ComponentPose=@n1, Alpha=1, Translation="0,0,0", Rotation="0,-40,0", Scale="1,1,1") {
    rotationMode=BMM_Additive
    rotationSpace=BCS_BoneSpace
    alphaBoolBlend {blendOption=Linear}
    boneToModify {boneName=r_wing_feather_mid_01}
    rotation {yaw=-40}
   }
   n3 = "Transform (Modify) Bone - Bone: r_wing_feather_bot_01"(ComponentPose=@n2, Alpha=1, Translation="0,0,0", Rotation="0,-20,0", Scale="1,1,1") {
    rotationMode=BMM_Additive
    rotationSpace=BCS_BoneSpace
    alphaBoolBlend {blendOption=Linear}
    boneToModify {boneName=r_wing_feather_bot_01}
    rotation {yaw=-20}
   }
   n4 = "Transform (Modify) Bone - Bone: l_wing_feathel_top_01"(ComponentPose=@n3, Alpha=1, Translation="0,0,0", Rotation="0,-60,0", Scale="1,1,1") {
    rotationMode=BMM_Additive
    rotationSpace=BCS_BoneSpace
    alphaBoolBlend {blendOption=Linear}
    boneToModify {boneName=l_wing_feathel_top_01}
    rotation {yaw=-60}
   }
   n5 = "Transform (Modify) Bone - Bone: l_wing_feathel_mid_01"(ComponentPose=@n4, Alpha=1, Translation="0,0,0", Rotation="0,-40,0", Scale="1,1,1") {
    rotationMode=BMM_Additive
    rotationSpace=BCS_BoneSpace
    alphaBoolBlend {blendOption=Linear}
    boneToModify {boneName=l_wing_feathel_mid_01}
    rotation {yaw=-40}
   }
   n6 = "Transform (Modify) Bone - Bone: l_wing_feathel_bot_01"(ComponentPose=@n5, Alpha=1, Translation="0,0,0", Rotation="0,-20,0", Scale="1,1,1") {
    rotationMode=BMM_Additive
    rotationSpace=BCS_BoneSpace
    alphaBoolBlend {blendOption=Linear}
    boneToModify {boneName=l_wing_feathel_bot_01}
    rotation {yaw=-20}
   }
   Output Animation Pose(Result=Component To Local(ComponentPose=@n6))
  ]
 }
 Targeting {
  path=AnimGraph.AnimGraphNode_StateMachine_0.wingPosition.AnimStateNode_3.Targeting
  nodes [Output Animation Pose(Result=Local Space Ref Pose())]
 }
 Diving {
  path=AnimGraph.AnimGraphNode_StateMachine_0.wingPosition.AnimStateNode_4.Diving
  nodes [
   n1 = "Transform (Modify) Bone - Bone: r_wing_feather_top_01"(ComponentPose=Local To Component(LocalPose=Local Space Ref Pose()), Alpha=1, Translation="0,0,0", Rotation="0,15,0", Scale="1,1,1") {
    rotationMode=BMM_Additive
    rotationSpace=BCS_BoneSpace
    alphaBoolBlend {blendOption=Linear}
    boneToModify {boneName=r_wing_feather_top_01}
    rotation {yaw=15}
   }
   n2 = "Transform (Modify) Bone - Bone: r_wing_feather_mid_01"(ComponentPose=@n1, Alpha=1, Translation="0,0,0", Rotation="0,15,0", Scale="1,1,1") {
    rotationMode=BMM_Additive
    rotationSpace=BCS_BoneSpace
    alphaBoolBlend {blendOption=Linear}
    boneToModify {boneName=r_wing_feather_mid_01}
    rotation {yaw=15}
   }
   n3 = "Transform (Modify) Bone - Bone: r_wing_feather_bot_01"(ComponentPose=@n2, Alpha=1, Translation="0,0,0", Rotation="0,15,0", Scale="1,1,1") {
    rotationMode=BMM_Additive
    rotationSpace=BCS_BoneSpace
    alphaBoolBlend {blendOption=Linear}
    boneToModify {boneName=r_wing_feather_bot_01}
    rotation {yaw=15}
   }
   n4 = "Transform (Modify) Bone - Bone: l_wing_feathel_top_01"(ComponentPose=@n3, Alpha=1, Translation="0,0,0", Rotation="0,15,0", Scale="1,1,1") {
    rotationMode=BMM_Additive
    rotationSpace=BCS_BoneSpace
    alphaBoolBlend {blendOption=Linear}
    boneToModify {boneName=l_wing_feathel_top_01}
    rotation {yaw=15}
   }
   n5 = "Transform (Modify) Bone - Bone: l_wing_feathel_mid_01"(ComponentPose=@n4, Alpha=1, Translation="0,0,0", Rotation="0,15,0", Scale="1,1,1") {
    rotationMode=BMM_Additive
    rotationSpace=BCS_BoneSpace
    alphaBoolBlend {blendOption=Linear}
    boneToModify {boneName=l_wing_feathel_mid_01}
    rotation {yaw=15}
   }
   n6 = "Transform (Modify) Bone - Bone: l_wing_feathel_bot_01"(ComponentPose=@n5, Alpha=1, Translation="0,0,0", Rotation="0,15,0", Scale="1,1,1") {
    rotationMode=BMM_Additive
    rotationSpace=BCS_BoneSpace
    alphaBoolBlend {blendOption=Linear}
    boneToModify {boneName=l_wing_feathel_bot_01}
    rotation {yaw=15}
   }
   Output Animation Pose(Result=Component To Local(ComponentPose=@n6))
  ]
 }
 Transition {
  path=AnimGraph.AnimGraphNode_StateMachine_0.wingPosition.AnimStateTransitionNode_1.Transition
  nodes [Result(bCanEnterTransition=Get bIsOnGround() -> (bIsOnGround:bool))]
 }
 Transition {
  path=AnimGraph.AnimGraphNode_StateMachine_0.wingPosition.AnimStateTransitionNode_2.Transition
  nodes [Result(bCanEnterTransition=Get bIsParachuteOpened())]
 }
 Transition {
  path=AnimGraph.AnimGraphNode_StateMachine_0.wingPosition.AnimStateTransitionNode_3.Transition
  nodes [Result(bCanEnterTransition=Get bIsSkydiveDiveMode())]
 }
 Transition {
  path=AnimGraph.AnimGraphNode_StateMachine_0.wingPosition.AnimStateTransitionNode_4.Transition
  nodes [Result(bCanEnterTransition=NOT Boolean(A=Get bIsSkydiveDiveMode()))]
 }
 Transition {
  path=AnimGraph.AnimGraphNode_StateMachine_0.wingPosition.AnimStateTransitionNode_5.Transition
  nodes [Result(bCanEnterTransition=Get bIsSkydiveDiveMode())]
 }
 Transition {
  path=AnimGraph.AnimGraphNode_StateMachine_0.wingPosition.AnimStateTransitionNode_6.Transition
  nodes [Result(bCanEnterTransition=Get bIsTargeting())]
 }
 Transition {
  path=AnimGraph.AnimGraphNode_StateMachine_0.wingPosition.AnimStateTransitionNode_7.Transition
  nodes [Result(bCanEnterTransition=NOT Boolean(A=Get bIsTargeting()))]
 }
 Transition {
  path=AnimGraph.AnimGraphNode_StateMachine_0.wingPosition.AnimStateTransitionNode_8.Transition
  nodes [Result(bCanEnterTransition=Get bIsJumping())]
 }
 Transition {
  path=AnimGraph.AnimGraphNode_StateMachine_0.wingPosition.AnimStateTransitionNode_9.Transition
  nodes [Result(bCanEnterTransition=NOT Boolean(A=Get bIsJumping()))]
 }
 EventGraph {nodes [
   TryGetPawnOwner()
   Event BlueprintUpdateAnimation()
  ]}
]
variables [NewVar_0 {type=number}]
